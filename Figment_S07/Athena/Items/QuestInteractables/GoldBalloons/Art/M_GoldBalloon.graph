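# Subsurface material for the gold balloon quest interactable, with vertex-color tinting, an optional decal and a Fresnel rim.
# Base color blends three color parameters by vertex red and green over the party balloon texture, also feeding subsurface color.
# World position offset combines slow actor-centered wobble with Pivot Painter wind sway, randomized per element and masked by vertex green.
# Static switches add a Fresnel-based glow (default on) and a panning camera-space sweep band (default off).
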Material /Figment_S07/Athena/Items/QuestInteractables/GoldBalloons/Art/M_GoldBalloon
graph {
 properties {
  pixelDepthOffsetMode=PDOM_Legacy
  previewMesh=&/SM_GoldBalloon
  shadingModel=MSM_Subsurface
  usedShadingModels=Subsurface
 }
 nodes [
  n1 = Multiply(A=Time(), B=ScalarParameter() {defaultValue=0.17, parameterName=SpeedBallon_2})
  n2 = AppendVector(A=Sine(None=@n1), B=AppendVector(A=Sine(None=@n1), B=Cosine(None=@n1)))
  n3 = Sine(None=Multiply(A=Time(), B=ScalarParameter() {defaultValue=0.09, parameterName=OverallRotationSpeed}))
  n4 = ScalarParameter() {defaultValue=0.01, parameterName=OverallRotation}
  n5 = Constant3Vector()
  n6 = ActorPositionWS()
  n7 = WorldPosition() {worldPositionShaderOffset=WPT_ExcludeAllShaderOffsets}
  n8 = Normalize(VectorInput=Transform(None=VectorParameter() {parameterName=RotaionAxis, defaultValue=<rgba 0 0 1 1>}.RGB) {transformSourceType=TRANSFORMSOURCE_Local})
  n9 = Time()
  n10 = MaterialFunctionCall() {materialFunction=/Engine/Functions/Engine_MaterialFunctions02/PivotPainter/PivotPainter_PerObjectData}
  n11 = ScalarParameter() {defaultValue=1.5613, parameterName=OffsetVariation}
  n12 = Multiply(
   A=Add(A=@n9, B=Multiply(A=@n10.Random Value Per Element, B=@n11))
   B=ScalarParameter() {defaultValue=0.4, parameterName=WindRotation_Speed_1}
  )
  n13 = ScalarParameter() {defaultValue=2.2338, parameterName=OffsetVariation_1}
  n14 = Multiply(A=Add(A=Multiply(A=@n10.Random Value Per Element, B=@n13), B=@n9), B=ScalarParameter() {defaultValue=0.4, parameterName=WindRotation_Speed})
  n15 = Multiply(A=Add(A=Sine(None=@n12), B=Sine(None=@n14)), B=ScalarParameter() {defaultValue=0.02, parameterName=WindRotation_strength})
  Add(
   A=RotateAboutAxis(NormalizedRotationAxis=Normalize(VectorInput=@n2), RotationAngle=Multiply(A=@n3, B=@n4), PivotPoint=Add(A=@n5, B=@n6), Position=@n7.XYZ)
   B=Multiply(A=RotateAboutAxis(NormalizedRotationAxis=@n8, RotationAngle=@n15, PivotPoint=@n10.Pivot Position, Position=@n7.XYZ), B=VertexColor().G)
  )
  n16 = VertexNormalWS()
  n17 = MaterialFunctionCall(
   Brightness=Constant()
   Exponent=Constant() {r=4}
   Camera Vector=Constant3Vector() {constant=<rgba 0 0 -1 0>}
   Vertex / Pixel Normal=Normalize(VectorInput=Transform(None=@n16) {transformSourceType=TRANSFORMSOURCE_World, transformType=TRANSFORM_View})
  ) {materialFunction=&MF_Fresnel_FX}
  n18 = VectorParameter() {parameterName=GlowColour, defaultValue=<rgba 10 3.0638 0.4416 0>}
  n19 = Multiply(
   A=Subtract(
    A=WorldPosition().XYZ
    B=MaterialFunctionCall() {materialFunction=&ComponentPivotLocation}.Result
   )
   B=ScalarParameter() {defaultValue=0.001, parameterName=WorldTileScale}
  )
  n20 = Normalize(VectorInput=Transform(None=VectorParameter() {parameterName=SweepDirectionCamSpace, defaultValue=<rgba 0.45 0.0412 0.3193 0>}.RGB) {transformSourceType=TRANSFORMSOURCE_Camera})
  n21 = ComponentMask(None=Multiply(A=DotProduct(A=@n19, B=@n20), B=@n20)) {b=true, r=true}
  n22 = ScalarParameter() {defaultValue=0.1, parameterName=PanSpeed}
  n23 = Frac(None=ComponentMask(None=Panner(Coordinate=@n21, Speed=@n22) {speedX=0.2, speedY=0.2}) {g=true})
  n24 = Add(A=@n23) {constB=0.1}
  n25 = OneMinus(None=Multiply(A=Subtract(A=@n23) {constB=0.9}) {constB=10})
  n26 = MaterialFunctionCall() {materialFunction=&MF_Fresnel_FX}
  n27 = Multiply(A=Multiply(A=Floor(None=@n24), B=Power(Base=@n25) {constExponent=3}), B=Saturate(None=Add(A=@n26.Fresnel Inverted) {constB=0.1}))
  n28 = VectorParameter() {parameterName=SweepColour, defaultValue=<rgba 2 1.6876 0.8067 0>}
  Add(
   A=StaticSwitchParameter(True=Multiply(A=@n17.Fresnel Inverted, B=@n18.RGB), False=Constant()) {defaultValue=true, parameterName=UseOuterGlow}
   B=StaticSwitchParameter(True=Multiply(A=@n27, B=@n28.RGB), False=Constant()) {parameterName=UseSweep}
  )
  n29 = ScalarParameter() {defaultValue=4, parameterName=FresnelPower}
  n30 = VertexColor()
  n31 = LinearInterpolate(
   A=VectorParameter() {parameterName=Color_1, defaultValue=<rgba 0.2767 0.6042 0.2197 1>}.RGB
   B=VectorParameter() {parameterName=Color_2, defaultValue=<rgba 1 0.5647 0.15 1>}.RGB
   Alpha=@n30.R
  )
  n32 = VectorParameter() {parameterName=Color_3, defaultValue=<rgba 0 0.235 0.9635 1>}
  n33 = MaterialFunctionCall(
   UVs=TextureCoordinate() {coordinateIndex=1}
   Texture Scale=AppendVector(
    A=ScalarParameter() {defaultValue=1, group=Decal, parameterName=DecalScale_X}
    B=ScalarParameter() {defaultValue=1, group=Decal, parameterName=DecalScale_Y}
   )
  ) {materialFunction=&ScaleUVsByCenter}
  n34 = TextureSampleParameter2D(UVs=@n33.UVs) {group=Decal, parameterName=DecalTexture, texture=/Game/Environments/Props/BirthdayBalloons/Textures/T_NoDecal}
  n35 = LinearInterpolate(
   A=Multiply(
    A=LinearInterpolate(A=@n31, B=@n32.RGB, Alpha=@n30.G)
    B=TextureSample() {texture=/Game/Environments/Props/BirthdayBalloons/Textures/T_PartyBalloon_D}.RGB
   )
   B=@n34.RGB
   Alpha=Multiply(A=@n34.A, B=@n33.0-1 mask)
  )
  Add(
   A=Multiply(
    A=Fresnel(ExponentIn=@n29) {baseReflectFraction=0}
    B=VectorParameter() {parameterName=FresnelColor, defaultValue=<rgba 0.5 0.49 0.474 1>}.RGB
   )
   B=@n35
  )
  Constant() {r=1}
  Constant() {r=0.1}
  Constant() {r=0.5}
  Multiply(A=@n35, B=ScalarParameter() {defaultValue=0.45, parameterName=SSSBrightness})
  n36 = AppendVector(A=TextureCoordinate() {coordinateIndex=2}, B=ComponentMask(None=TextureCoordinate() {coordinateIndex=3}))
  n37 = Constant3Vector() {constant=<rgba 0 -1 0 0>}
  Normalize(VectorInput=TransformPosition(None=Add(A=@n36, B=@n37)) {transformType=TRANSFORMPOSSOURCE_World})
  ScalarParameter() {defaultValue=0.5, parameterName=SSSDistance}
  TextureSample() {texture=/Game/Environments/Props/BirthdayBalloons/Textures/T_PartyBalloon_M}
  TextureSample() {samplerType=SAMPLERTYPE_Normal, texture=/Game/Environments/Props/BirthdayBalloons/Textures/T_PartyBalloon_N}
  VectorParameter() {parameterName=RotationCenter, defaultValue=<rgba 0 0 0 1>}
  WorldPosition()
 ]
}
